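# Physics asset for the Uproar backpack's DBNO (downed) variant, simulating a small hanging figure with limbs, three-segment ears and a scarf.
# Capsule bodies cover the pack and dyn_ bones, while a large, thin box sits on root.
# The pack, constrained to root, is the hub for pelvis/legs, arms, head/ears and scarf chains, all with angular limits.
# A separate three-body spine chain of oversized capsules also hangs off the pack, with free swing and twist.
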PhysicsAsset /BRCosmetics/Accessories/FORT_Backpacks/Backpack_F_MED_Uproar/Meshes/F_MED_Uproar_Pack_DBNO_Physics
bodies {
 dyn_calf_l [{
   shapeType=Capsule
   length=3.5907
   radius=1.7574
   shapeName=dyn_calf_l_capsule
   center=<3.5941 -0.08 -0.1113>
   rotation=<rot 0 -94.0205 -88.7397>
  }]
 dyn_calf_r [{
   shapeType=Capsule
   length=3.3814
   radius=1.882
   shapeName=dyn_calf_r_capsule
   center=<3.8682 -0.1556 0.071>
   rotation=<rot 0 -93.8394 -87.6268>
  }]
 dyn_ear_01_l [{
   shapeType=Capsule
   length=2.4425
   radius=1.9512
   shapeName=dyn_ear_01_l_capsule
   center=<2.1886 -0.1295 0.0655>
   rotation=<rot 0 95.275 -89.7305>
  }]
 dyn_ear_01_r [{
   shapeType=Capsule
   length=1.7768
   radius=1.6059
   shapeName=dyn_ear_01_r_capsule
   center=<1.7094 -0.1281 0.0403>
   rotation=<rot 0 96.9053 -84.9944>
  }]
 dyn_ear_02_l [{
   shapeType=Capsule
   length=3.5834
   radius=2.6108
   shapeName=dyn_ear_02_l_capsule
   center=<2.4947 0.0737 0.0447>
   rotation=<rot 0 -119.9552 -89.5503>
  }]
 dyn_ear_02_r [{
   shapeType=Capsule
   length=2.4417
   radius=1.9395
   shapeName=dyn_ear_02_r_capsule
   center=<1.7275 -0.1519 -0.0037>
   rotation=<rot 0 -166.3415 -85.3933>
  }]
 dyn_ear_03_l [{
   shapeType=Capsule
   length=2.6815
   radius=1.5853
   shapeName=dyn_ear_03_l_capsule
   center=<2.1861 -0.2663 0.0263>
   rotation=<rot 0 -100.6625 -88.8696>
  }]
 dyn_ear_03_r [{
   shapeType=Capsule
   length=2.0647
   radius=1.9952
   shapeName=dyn_ear_03_r_capsule
   center=<1.866 -0.3521 0.0496>
   rotation=<rot 0.7167 -13.2545 -90>
  }]
 dyn_head [{shapeType=Capsule, length=3, radius=3.7, shapeName=dyn_head_capsule, center=<3.6811 -1.4664 -0.049>, rotation=<rot 0 150.0715 -88.1333>}]
 dyn_lowerarm_l [{shapeType=Capsule, length=3, radius=1, shapeName=dyn_lowerarm_l_capsule, center=<2.5 0 0>, rotation=<rot 90 0 0>}]
 dyn_lowerarm_r [{shapeType=Capsule, length=3, radius=1, shapeName=dyn_lowerarm_r_capsule, center=<2.5 0 0>, rotation=<rot 90 0 0>}]
 dyn_pelvis [{shapeType=Capsule, length=2, radius=3, shapeName=dyn_pelvis_capsule, center=<1.8825 0.4382 -0.004>}]
 dyn_scarf_01 [{shapeType=Capsule, length=1, radius=0.4, shapeName=dyn_scarf_01_capsule, center=<1.1041 0 0>}]
 dyn_scarf_02 [{shapeType=Capsule, length=2, radius=0.5, shapeName=dyn_scarf_02_capsule, center=<1.4497 0 0>}]
 dyn_spine_01 [{shapeType=Capsule, length=15, radius=10, shapeName=dyn_spine_01_capsule, center=<5 0 0>}]
 dyn_spine_02 [{shapeType=Capsule, length=15, radius=10, shapeName=dyn_spine_02_capsule, center=<5 0 0>}]
 dyn_spine_03 [{shapeType=Capsule, length=15, radius=10, shapeName=dyn_spine_03_capsule, center=<5 0 0>}]
 dyn_thigh_l [{
   shapeType=Capsule
   length=3.5623
   radius=1.6004
   shapeName=dyn_thigh_l_capsule
   center=<3.4397 0.0246 0.0129>
   rotation=<rot 0 -90.8046 -89.4864>
  }]
 dyn_thigh_r [{
   shapeType=Capsule
   length=3.4441
   radius=1.6474
   shapeName=dyn_thigh_r_capsule
   center=<3.3107 -0.0513 0.0688>
   rotation=<rot 0 -89.3401 -89.1129>
  }]
 dyn_upperarm_l [{shapeType=Capsule, length=3.5, radius=1, shapeName=dyn_upperarm_l_capsule, center=<2.5 0 0>, rotation=<rot 90 0 0>}]
 dyn_upperarm_r [{shapeType=Capsule, length=3.5, radius=1, shapeName=dyn_upperarm_r_capsule, center=<2.5 0 0>, rotation=<rot 90 0 0>}]
 pack [{shapeType=Capsule, length=4.6672, radius=2.3, shapeName=pack_capsule, center=<0 -3.0203 -4.3786>, rotation=<rot 0 0 7.0021>}]
 root [{shapeType=Box, extentX=30, extentY=5, extentZ=49, shapeName=root_box, center=<0 2.5 -3>, rotation=<rot 0 0 -5.0012>}]
}
constraints [
 {bone1Name=pack, bone2Name=root, swing1LimitDegrees=45, swing2LimitDegrees=45, twistLimitDegrees=45}
 {bone1Name=dyn_pelvis, bone2Name=pack, swing1LimitDegrees=10, swing2LimitDegrees=15, twistLimitDegrees=5}
 {bone1Name=dyn_thigh_l, bone2Name=dyn_pelvis, swing1LimitDegrees=25, swing2LimitDegrees=15, twistLimitDegrees=5}
 {bone1Name=dyn_calf_l, bone2Name=dyn_thigh_l, swing1LimitDegrees=20, swing2LimitDegrees=15, twistLimitDegrees=10}
 {bone1Name=dyn_thigh_r, bone2Name=dyn_pelvis, swing1LimitDegrees=25, swing2LimitDegrees=15, twistLimitDegrees=5}
 {bone1Name=dyn_calf_r, bone2Name=dyn_thigh_r, swing1LimitDegrees=20, swing2LimitDegrees=15, twistLimitDegrees=10}
 {bone1Name=dyn_upperarm_l, bone2Name=pack, swing1LimitDegrees=45, swing2LimitDegrees=75, twistLimitDegrees=15}
 {bone1Name=dyn_lowerarm_l, bone2Name=dyn_upperarm_l, swing1LimitDegrees=45, swing2LimitDegrees=45, twistLimitDegrees=15}
 {bone1Name=dyn_upperarm_r, bone2Name=pack, swing1LimitDegrees=45, swing2LimitDegrees=75, twistLimitDegrees=15}
 {bone1Name=dyn_lowerarm_r, bone2Name=dyn_upperarm_r, swing1LimitDegrees=45, swing2LimitDegrees=45, twistLimitDegrees=15}
 {bone1Name=dyn_head, bone2Name=pack, swing1LimitDegrees=15, swing2LimitDegrees=15, twistLimitDegrees=5}
 {bone1Name=dyn_ear_01_l, bone2Name=dyn_head, swing1LimitDegrees=5, swing2LimitDegrees=15, twistLimitDegrees=5}
 {bone1Name=dyn_ear_02_l, bone2Name=dyn_ear_01_l, swing1LimitDegrees=15, swing2LimitDegrees=30, twistLimitDegrees=15}
 {bone1Name=dyn_ear_03_l, bone2Name=dyn_ear_02_l, swing1LimitDegrees=15, swing2LimitDegrees=30, twistLimitDegrees=15}
 {bone1Name=dyn_ear_01_r, bone2Name=dyn_head, swing1LimitDegrees=5, swing2LimitDegrees=15, twistLimitDegrees=5}
 {bone1Name=dyn_ear_02_r, bone2Name=dyn_ear_01_r, swing1LimitDegrees=15, swing2LimitDegrees=30, twistLimitDegrees=15}
 {bone1Name=dyn_ear_03_r, bone2Name=dyn_ear_02_r, swing1LimitDegrees=15, swing2LimitDegrees=30, twistLimitDegrees=15}
 {bone1Name=dyn_scarf_01, bone2Name=pack, swing1LimitDegrees=30, swing2LimitDegrees=30, twistLimitDegrees=15}
 {bone1Name=dyn_scarf_02, bone2Name=dyn_scarf_01, swing1LimitDegrees=45, swing2LimitDegrees=45, twistLimitDegrees=15}
 {
  bone1Name=dyn_spine_01
  bone2Name=pack
  swing1LimitDegrees=45
  swing1Motion=Free
  swing2LimitDegrees=45
  swing2Motion=Free
  twistLimitDegrees=45
  twistMotion=Free
 }
 {
  bone1Name=dyn_spine_02
  bone2Name=dyn_spine_01
  swing1LimitDegrees=45
  swing1Motion=Free
  swing2LimitDegrees=45
  swing2Motion=Free
  twistLimitDegrees=45
  twistMotion=Free
 }
 {
  bone1Name=dyn_spine_03
  bone2Name=dyn_spine_02
  swing1LimitDegrees=45
  swing1Motion=Free
  swing2LimitDegrees=45
  swing2Motion=Free
  twistLimitDegrees=45
  twistMotion=Free
 }
]
